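# Unlit, two-sided translucent material for scythe weapon ribbon effects, flagged for Niagara ribbons and beam trails.
# Particle color RGB feeds the shared FX emissive adjustments function for the glow.
# An opacity mask multiplies particle alpha by a slowly panning, slightly lifted caustics texture and procedural UV gradients.
# The gradients ramp in quickly then taper along U and soften both V edges, and depth fade softens geometry intersections.
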Material /Game/Weapons/FORT_Melee/Effects/Scythe/M_Syth_Ribbon
graph {
 properties {
  bUsedWithBeamTrails=true
  bUsedWithNiagaraRibbons=true
  blendMode=BLEND_Translucent
  pixelDepthOffsetMode=PDOM_Legacy
  shadingModel=MSM_Unlit
  twoSided=true
  usedShadingModels=Unlit
 }
 nodes [
  Constant() {r=1}
  n1 = ParticleColor()
  MaterialFunctionCall(Emissive=@n1.RGB) {materialFunction=&MF_FX_EmissiveAdjustments}
  n2 = Panner(Coordinate=TextureCoordinate() {uTiling=0.3}) {speedX=0.1}
  n3 = Saturate(None=Add(A=TextureSample(UVs=@n2) {texture=&T_ShieldCaustics}.R) {constB=0.12})
  n4 = Saturate(None=ComponentMask(None=TextureCoordinate()) {r=true})
  n5 = ComponentMask(None=TextureCoordinate()) {r=true}
  n6 = Multiply(A=Power(Base=OneMinus(None=@n4)), B=Saturate(None=Multiply(A=@n5) {constB=8}))
  n7 = ComponentMask(None=TextureCoordinate()) {g=true}
  n8 = Multiply(A=@n7) {constB=8}
  n9 = Multiply(A=OneMinus(None=@n7)) {constB=8}
  n10 = Power(Base=Multiply(A=Saturate(None=@n8), B=Saturate(None=@n9)))
  Multiply(A=Multiply(A=Multiply(A=@n1.A, B=@n3), B=Multiply(A=@n6, B=@n10)), B=DepthFade(FadeDistance=Constant() {r=8}))
 ]
}
